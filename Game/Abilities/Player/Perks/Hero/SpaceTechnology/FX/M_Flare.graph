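# Translucent unlit particle material that draws a spiky flare sprite for the Space Technology hero perk effects.
# The flare texture's green channel blends particle color toward a CenterColor parameter, an overbright cyan by default, and the result passes through a shared emissive-adjustment function.
# Particle alpha, the texture's red channel and a parameterized depth fade multiply into a soft-edged mask that avoids hard intersections with geometry.
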Material /Game/Abilities/Player/Perks/Hero/SpaceTechnology/FX/M_Flare
graph {
 properties {
  bUsedWithNiagaraSprites=true
  bUsedWithParticleSprites=true
  blendMode=BLEND_Translucent
  pixelDepthOffsetMode=PDOM_Legacy
  shadingModel=MSM_Unlit
  usedShadingModels=Unlit
 }
 nodes [
  n1 = ParticleColor()
  n2 = TextureSample() {texture=/Game/Effects/Fort_Effects/Textures/Flares/T_Flare_Spikeball}
  MaterialFunctionCall(Emissive=LinearInterpolate(A=@n1.RGB, B=VectorParameter() {parameterName=CenterColor, defaultValue=<rgba 0.75 4 3 1>}.RGB, Alpha=@n2.G)) {materialFunction=&MF_FX_EmissiveAdjustments}
  n3 = ScalarParameter() {defaultValue=100, parameterName=DF_FD}
  Multiply(A=@n1.A, B=Multiply(A=@n2.R, B=DepthFade(FadeDistance=@n3)))
 ]
}
